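# Editor utility widget that finds and thins overlapping instances in instanced foliage clusters, on selected or all loaded actors.
# Per cluster, it pairwise-compares instance distances against radii from scaled mesh Z extent and a threshold, then picks each overlap group's smallest.
# Testing mode flags picks through per-instance custom data for material debugging; otherwise, after confirmation, it deletes them.
# A helper button selects clusters within a hardcoded XY distance of the selection or camera.
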EditorUtilityWidgetBlueprint /Game/ArtTools/Tech_Art/InstancedMeshesCluster/ClusterDentiyHelper/EUW_ClusterDensityHelper
graphs [
 DoOverlapLogic {nodes [
   n1 = DoOverlapLogic!() -> (then, Actors:Array of BP Instanced Cluster Object References) {is=title K2Node_FunctionEntry}
   > n2 = For Each Loop(Array:Array of BP Instanced Cluster Object References=@n1) -> (Array Element:BP Instanced Cluster Object Reference)
   n3 = "To String (Integer)"(InInt=Get l_total() -> (l_total:int=0))
   "SetText (Text)"(
    execute=@n2.Completed
    self=Get CommonTextBlock_0() -> (CommonTextBlock_0:Common Text Object Reference)
    InText="To Text (String)"(InString=Append!(A="Report: Found a total of ", B=@n3, C:String=" Overlapping Instances to remove.") -> (ReturnValue))
   )
   n4 = HandleCluster!(execute=@n2.LoopBody, self, actor:Actor Object Reference=@n2) -> (then, totalHidden:int=0)
   > Set l_total(l_total:int=int + int(A=@n4, B=Get l_total() -> (l_total=0))) -> (Output_Get:int=0)
  ]}
 EventGraph {nodes [
   "On Clicked (TA_Generic01_Button)"!() -> (OutputDelegate, then, widget:User Widget Object Reference, event:Pointer Event Structure) {is=title K2Node_ComponentBoundEvent}
   > n1 = Branch(Condition=Get IsTesting() -> (IsTesting:bool))
   n2 = GetActorsToRunOn!(execute=[@n1.then, @n7.Ok, @n7.Yes], self) -> (then, actors:Array of BP Instanced Cluster Object References)
   > n3 = IfDebugPreUpdateActors!(execute, self, Array:Array of BP Instanced Cluster Object References=@n2) -> (then, NeedDelay:bool)
   > n4 = Branch(Condition=@n3)
   > n5 = RetriggerableDelay!(execute, WorldContextObject:Object Reference, Duration=0.01) -> (then)
   DoOverlapLogic!(execute=[@n5.then, @n4.else], self, Actors:Array of BP Instanced Cluster Object References=@n2) -> (then) {is=title K2Node_CallFunction}
   n6 = Show Message Dialog(
    execute=@n1.else
    Title=MakeLiteralText(Value="INVTEXT(\"Run Foliage Cluster Instance Removal\")")
    Message=MakeLiteralText(Value="INVTEXT(\"This tool runs on all loaded Foliage Clusters and tries to remove overlapping instances.\\r\\nYou are about to possibly delete hundreds of foliage instances.  Are you sure you want to run? \")")
    MessageType=OkCancel
    DefaultValue=Cancel
    MessageCategory=Warning
   )
   > n7 = Switch on EAppReturnType(Selection=@n6)
   "On Clicked (btnSelect)"!() -> (OutputDelegate, then, widget:User Widget Object Reference, event:Pointer Event Structure) {is=title K2Node_ComponentBoundEvent}
   > SelectNearby()
   "On Property Changed (DetailsView_0)"!() -> (OutputDelegate, then, PropertyName:Name=None) {is=title K2Node_ComponentBoundEvent}
   > ValidateVisuals()
   HandleSelectionChanged!() -> (OutputDelegate, then, SelectionSet:Typed Element Selection Set Object Reference) {is=title K2Node_CustomEvent}
   > ValidateSelectText()
   Event Destruct()
   > n8 = GetSelectionSet(self=Get LevelEditorSubsystem())
   > Unbind Event from On Selection Change(self:Typed Element Selection Set Object Reference=@n8, Delegate=Create Event()) {is=title K2Node_RemoveDelegate}
   Event Construct()
   > SetObject(self=Get DetailsView_0() -> (DetailsView_0:Details View Object Reference), NewObject=Self-Reference())
   > ValidateVisuals()
   > n9 = GetSelectionSet(self=Get LevelEditorSubsystem())
   > Bind Event to On Selection Change(self:Typed Element Selection Set Object Reference=@n9, Delegate=Create Event()) {is=title K2Node_AddDelegate}
  ]}
 GetActorsToRunOn {nodes [
   GetActorsToRunOn() {is=title K2Node_FunctionEntry}
   > n1 = Branch(Condition=Get RunOnlyOnSelectedActors() -> (RunOnlyOnSelectedActors:bool))
   > n2 = GetSelectedLevelActors(self=Get EditorActorSubsystem())
   > n3 = Filter by Class!(
    execute
    TargetArray:Array of Object References=@n2
    ObjectClass:Object Class Reference=/Game/ArtTools/Tech_Art/InstancedMeshesCluster/BP_InstancedCluster.BP_InstancedCluster_C
    FilterType:EEditorScriptingFilterType Enum
   ) -> (then, ReturnValue:Array of BP Instanced Cluster Object References)
   > n4 = Set l_actors(l_actors:Array of BP Instanced Cluster Object References=@n3) -> (Output_Get:Array of BP Instanced Cluster Object References)
   Return Node!(
    execute=[@n6.then, @n4.then]
    actors:Array of BP Instanced Cluster Object References=Get l_actors() -> (l_actors:Array of BP Instanced Cluster Object References)
   )
   n5 = GetAllActorsOfClass!(
    execute=@n1.else
    WorldContextObject:Object Reference
    ActorClass:Actor Class Reference=&BP_InstancedCluster_C
   ) -> (then, OutActors:Array of BP Instanced Cluster Object References)
   > n6 = Set l_actors(l_actors=@n5)
  ]}
 HandleCluster {nodes [
   n1 = HandleCluster!() -> (then, actor:Actor Object Reference) {is=title K2Node_FunctionEntry}
   > Set l_inActor(l_inActor:Actor Object Reference=@n1) -> (Output_Get:Actor Object Reference)
   > n2 = Sequence!(execute) -> (then_0, then_1, then_2:Exec, then_3:Exec, then_4:Exec)
   > n3 = For Each Loop(Array:Array of Instanced Static Mesh Component Object References=Get Components By Class!(
     self:Actor Object Reference=Get l_inActor() -> (l_inActor:Actor Object Reference)
     ComponentClass:Actor Component Class Reference=/Script/Engine.InstancedStaticMeshComponent
    ) -> (ReturnValue:Array of Instanced Static Mesh Component Object References)) -> (Array Element:Instanced Static Mesh Component Object Reference)
   > n4 = Set l_comp(l_comp:Instanced Static Mesh Component Object Reference=@n3) -> (Output_Get:Instanced Static Mesh Component Object Reference)
   > Branch(Condition=Get bVisible!(self:Scene Component Object Reference=@n4) -> (bVisible:bool))
   > n5 = Set l_mesh(l_mesh:Static Mesh Object Reference=Get StaticMesh(self=Get l_comp() -> (l_comp:Instanced Static Mesh Component Object Reference))) -> (Output_Get:Static Mesh Object Reference)
   > Is Valid(InputObject=@n5)
   n6 = BreakBoxSphereBounds!(InBoxSphereBounds:"Box Sphere Bounds Structure (by ref)"=GetBounds!(self:Static Mesh Object Reference=Get l_mesh() -> (l_mesh:Static Mesh Object Reference)) -> (ReturnValue:Box Sphere Bounds Structure)) -> (Origin:Vector, BoxExtent:Vector, SphereRadius:float)
   > Set l_meshExtents(l_meshExtents:Vector=vector * vector(A=@n6.BoxExtent, B=Get OverlapPercentThreshold() -> (OverlapPercentThreshold:double)) -> (ReturnValue=~)) -> (Output_Get:Vector)
   n7 = BreakVector(InVec=@n6.BoxExtent)
   n8 = MakeVector2D(X=@n7, Y=@n7.Y)
   > Set l_meshRadius(l_meshRadius:double=float * float(A=Vector2D Length!(A:Vector 2D Structure=@n8) -> (ReturnValue:double), B=Get OverlapPercentThreshold()) {is=title K2Node_PromotableOperator}) -> (Output_Get:double)
   n9 = Get l_comp()
   > n10 = For Loop(LastIndex=int - int(A=GetInstanceCount!(self:Instanced Static Mesh Component Object Reference=@n9) -> (ReturnValue:int=0)))
   > Set l_curIndex(l_curIndex:int=@n10) -> (Output_Get:int=0)
   > Set l_curTransform(l_curTransform:Transform=GetInstanceTransform!(
     self:Instanced Static Mesh Component Object Reference=Get l_comp()
     InstanceIndex:int=Get l_curIndex() -> (l_curIndex:int=0)
     bWorldSpace:bool=true
    ) -> (OutInstanceTransform:Transform, ReturnValue:bool)) -> (Output_Get:Transform)
   n11 = Get l_comp()
   n12 = Get l_curIndex() -> (l_curIndex=0)
   n13 = String To Name(InString=Append(A=GetDisplayName(Object=@n11), B=_, C="To String (Integer)"(InInt=@n12)))
   n14 = Get!() -> (l_curTransform_Location:Vector, l_curTransform_Rotation:Rotator, l_curTransform_Scale:Vector)
   n15 = BreakVector(InVec=vector * vector(A=@n14.l_curTransform_Scale, B:Vector=Get l_meshExtents() -> (l_meshExtents:Vector)) -> (ReturnValue=~))
   n16 = "Absolute (Float)"(A=@n15.Z)
   > Add!(
    execute
    TargetMap:Map of Names to CDH Instance Info Struct Structures=Get l_instanceDataMap() -> (l_instanceDataMap:Map of Names to CDH Instance Info Struct Structures)
    Key:"Name (by ref)"=@n13
    Value:"CDH Instance Info Struct Structure (by ref)"=Make CDH Instance Info Struct!(
     component_2_#1:Instanced Static Mesh Component Object Reference=Get l_comp()
     instanceindex_5_#2:int=Get l_curIndex() -> (l_curIndex=0)
     transform_8_#3:Transform=Get l_curTransform() -> (l_curTransform:Transform)
     radius_11_#4:double=@n16
     uniqueInstanceName_14_#5:Name=@n13
    ) -> (CDH_InstanceInfo_Struct:CDH Instance Info Struct Structure) {is=title K2Node_MakeStruct}
   ) -> (then) {is=title K2Node_CallFunction}
   > UpdateCustomData!(
    execute
    self
    comp:Instanced Static Mesh Component Object Reference=Get l_comp()
    InstanceIndex:int=Get l_curIndex() -> (l_curIndex=0)
    CustomDataValue:float=0
    bMarkRenderStateDirty:bool=true
    CustomDataIndex:int=0
   ) -> (then)
   n17 = Values!(execute=@n2.then_1, TargetMap:Map of Names to CDH Instance Info Struct Structures=Get l_instanceDataMap()) -> (then, Values:Array of CDH Instance Info Struct Structures)
   > n18 = For Each Loop(Array:Array of CDH Instance Info Struct Structures=@n17) -> (Array Element:CDH Instance Info Struct Structure)
   > Set l_curIndex(l_curIndex=@n18.Array Index) -> (Output_Get=0)
   > Set l_sourceData(l_sourceData:CDH Instance Info Struct Structure=@n18) -> (Output_Get:CDH Instance Info Struct Structure)
   > n19 = Branch(Condition=Contains Item!(
     TargetSet:Set of Names=Get l_foundNames() -> (l_foundNames:Set of Names)
     ItemToFind:"Name (by ref)"=Get!() -> (
      l_sourceData_component_2_#1:Instanced Static Mesh Component Object Reference
      l_sourceData_instanceindex_5_#2:int
      l_sourceData_transform_8_#3:Transform
      l_sourceData_radius_11_#4:double=~
      l_sourceData_uniqueInstanceName_14_#5:Name
     ).l_sourceData_uniqueInstanceName_14_#5
    ) -> (ReturnValue:bool) {is=title K2Node_CallFunction})
   n20 = Values(execute=@n19.else, TargetMap=Get l_instanceDataMap())
   > n21 = For Loop with Break!(
    Execute:Exec
    FirstIndex:int=int + int(A=Get l_curIndex() -> (l_curIndex=0))
    LastIndex:int=Last Index!(TargetArray:Array of CDH Instance Info Struct Structures=@n20) -> (ReturnValue=0)
    Break:Exec=@n27.then_1
   ) -> (LoopBody:Exec, Index:int, Completed:Exec) {is=title K2Node_MacroInstance}
   > Set l_checkAgainstData(l_checkAgainstData:CDH Instance Info Struct Structure="Get (a ref)"(Array:Array of CDH Instance Info Struct Structures=@n20, Dimension 1=@n21) -> (Output:"CDH Instance Info Struct Structure (by ref)")) -> (Output_Get:CDH Instance Info Struct Structure)
   n22 = Break CDH Instance Info Struct!(CDH_InstanceInfo_Struct:"CDH Instance Info Struct Structure (by ref)"=Get l_checkAgainstData() -> (l_checkAgainstData:CDH Instance Info Struct Structure)) -> (
    component_2_#1:Instanced Static Mesh Component Object Reference
    instanceindex_5_#2:int
    transform_8_#3:Transform
    radius_11_#4:double=~
    uniqueInstanceName_14_#5:Name
   ) {is=title K2Node_BreakStruct}
   n23 = BreakTransform(InTransform=@n22.transform_8_#3)
   n24 = Break CDH Instance Info Struct(CDH_InstanceInfo_Struct=Get l_sourceData() -> (l_sourceData:CDH Instance Info Struct Structure)) -> (radius_11_#4=~)
   n25 = BreakTransform(InTransform=@n24.transform_8_#3)
   > Branch(Condition="float < float"(
     A="Distance (Vector)"!(V1:Vector=@n25, V2:Vector=@n23) -> (ReturnValue:double)
     B=float + float(A=@n24.radius_11_#4, B=@n22.radius_11_#4)
    ))
   > Set l_foundName(l_foundName:Name=@n24.uniqueInstanceName_14_#5) -> (Output_Get:Name=None)
   n26 = Get l_foundNames()
   > Add!(execute, TargetSet:Set of Names=@n26, NewItem:"Name (by ref)"=@n24.uniqueInstanceName_14_#5) -> (then)
   > Add(TargetSet=@n26, NewItem=@n22.uniqueInstanceName_14_#5)
   > n27 = Sequence!(execute) -> (then_0, then_1)
   > n28 = Branch(Condition=Contains!(
     TargetMap:Map of Names to CDH Instance Overlaps Struct Structures=Get l_overlapData() -> (l_overlapData:Map of Names to CDH Instance Overlaps Struct Structures)
     Key:"Name (by ref)"=Get l_foundName() -> (l_foundName:Name=None)
    ) -> (ReturnValue))
   > n29 = Set l_structToAddOverlaps(l_structToAddOverlaps:CDH Instance Info Struct Structure=Get l_checkAgainstData()) -> (Output_Get:CDH Instance Info Struct Structure)
   n30 = Set l_foundOverlapData(
    execute=[@n34.then, @n29.then]
    l_foundOverlapData:CDH Instance Overlaps Struct Structure=Find(
     TargetMap:Map of Names to CDH Instance Overlaps Struct Structures=Get l_overlapData()
     Key:"Name (by ref)"=Get l_foundName() -> (l_foundName=None)
    ) -> (Value:CDH Instance Overlaps Struct Structure)
   ) -> (Output_Get:CDH Instance Overlaps Struct Structure)
   > n31 = Set l_overlapStructs(l_overlapStructs:Array of CDH Instance Info Struct Structures=Break CDH Instance Overlaps Struct!(CDH_InstanceOverlaps_Struct:"CDH Instance Overlaps Struct Structure (by ref)"=@n30) -> (OverlappingStructs_3_#6:Array of CDH Instance Info Struct Structures) {is=title K2Node_BreakStruct}) -> (Output_Get:Array of CDH Instance Info Struct Structures)
   > Add Unique(
    TargetArray:Array of CDH Instance Info Struct Structures=@n31
    NewItem:"CDH Instance Info Struct Structure (by ref)"=Get l_structToAddOverlaps() -> (l_structToAddOverlaps:CDH Instance Info Struct Structure)
   )
   > n32 = Set members in CDH Instance Overlaps Struct!(
    execute:Exec
    StructRef:"CDH Instance Overlaps Struct Structure (by ref)"=Get l_foundOverlapData() -> (l_foundOverlapData:CDH Instance Overlaps Struct Structure)
    OverlappingStructs_3_#6:Array of CDH Instance Info Struct Structures=Get l_overlapStructs() -> (l_overlapStructs:Array of CDH Instance Info Struct Structures)
   ) -> (then:Exec, StructOut:"CDH Instance Overlaps Struct Structure (by ref)") {is=title K2Node_SetFieldsInStruct}
   > Add!(
    execute
    TargetMap:Map of Names to CDH Instance Overlaps Struct Structures=Get l_overlapData()
    Key:"Name (by ref)"=Get l_foundName() -> (l_foundName=None)
    Value:"CDH Instance Overlaps Struct Structure (by ref)"=@n32
   ) -> (then)
   Set l_foundName(execute=@n28.else, l_foundName=@n22.uniqueInstanceName_14_#5) -> (Output_Get=None)
   > n33 = Branch(Condition=Contains(TargetMap=Get l_overlapData(), Key=Get l_foundName() -> (l_foundName=None)))
   > n34 = Set l_structToAddOverlaps(l_structToAddOverlaps=Get l_sourceData())
   n35 = Get l_sourceData()
   n36 = Get l_checkAgainstData()
   Add(
    execute=@n33.else
    TargetMap=Get l_overlapData()
    Key=Get l_foundName() -> (l_foundName=None)
    Value=Make CDH Instance Overlaps Struct!(OverlappingStructs_3_#6:Array of CDH Instance Info Struct Structures=Make Array!("[0]":CDH Instance Info Struct Structure=@n35, "[1]":CDH Instance Info Struct Structure=@n36) -> (Array:Array of CDH Instance Info Struct Structures)) -> (CDH_InstanceOverlaps_Struct:CDH Instance Overlaps Struct Structure) {is=title K2Node_MakeStruct}
   )
   n37 = Values(execute=@n2.then_2, TargetMap:Map of Names to CDH Instance Overlaps Struct Structures=Get l_overlapData()) -> (Values:Array of CDH Instance Overlaps Struct Structures)
   > Set l_total(l_total=Length(TargetArray:Array of CDH Instance Overlaps Struct Structures=@n37)) -> (Output_Get=0)
   > n38 = For Each Loop(Array:Array of CDH Instance Overlaps Struct Structures=@n37) -> (Array Element:CDH Instance Overlaps Struct Structure)
   > Set l_smallestRadius(l_smallestRadius:double=9999999) -> (Output_Get:double)
   > n39 = For Each Loop(Array:Array of CDH Instance Info Struct Structures=Break CDH Instance Overlaps Struct(CDH_InstanceOverlaps_Struct=@n38)) -> (Array Element:CDH Instance Info Struct Structure)
   n40 = Get l_smallestData() -> (l_smallestData:CDH Instance Info Struct Structure)
   Add(
    execute=@n39.Completed
    TargetArray:Array of CDH Instance Info Struct Structures=Get l_finalInstanceDataToRemove() -> (l_finalInstanceDataToRemove:Array of CDH Instance Info Struct Structures)
    NewItem:"CDH Instance Info Struct Structure (by ref)"=@n40
   ) {is=title K2Node_CallArrayFunction}
   n41 = Break CDH Instance Info Struct(CDH_InstanceInfo_Struct=@n40) -> (radius_11_#4=~)
   > UpdateCustomData(
    comp=@n41
    InstanceIndex=@n41.instanceindex_5_#2
    CustomDataValue=1
    bMarkRenderStateDirty=true
    CustomDataIndex=0
   )
   n42 = Break CDH Instance Info Struct(CDH_InstanceInfo_Struct=@n39) -> (radius_11_#4=~)
   Branch(
    execute=@n39.LoopBody
    Condition="float < float"(A=@n42.radius_11_#4, B=Get l_smallestRadius() -> (l_smallestRadius:double))
   )
   > Set l_smallestRadius(l_smallestRadius=@n42.radius_11_#4)
   > Set l_smallestData(l_smallestData:CDH Instance Info Struct Structure=@n39) -> (Output_Get:CDH Instance Info Struct Structure)
   n43 = Branch(execute=@n2.then_3, Condition=Get IsTesting())
   n44 = For Each Loop(Exec=@n43.else, Array=Get l_finalInstanceDataToRemove())
   n45 = Get l_finalInstancesToClearMap() -> (l_finalInstancesToClearMap:Map of Instanced Static Mesh Component Object References to CDH Instance Indices Struct Structures)
   n46 = Keys!(execute=@n44.Completed, TargetMap:Map of Instanced Static Mesh Component Object References to CDH Instance Indices Struct Structures=@n45) -> (then, Keys:Array of Instanced Static Mesh Component Object References)
   > n47 = Values(TargetMap:Map of Instanced Static Mesh Component Object References to CDH Instance Indices Struct Structures=@n45) -> (Values:Array of CDH Instance Indices Struct Structures)
   > n48 = For Each Loop(Array:Array of Instanced Static Mesh Component Object References=@n46) -> (Array Element:Instanced Static Mesh Component Object Reference)
   > n49 = To Array!(
    execute
    A:Set of Integers=Break CDH Instance Indices Struct!(CDH_InstanceIndices_Struct:"CDH Instance Indices Struct Structure (by ref)"="Get (a ref)"(Array:Array of CDH Instance Indices Struct Structures=@n47, Dimension 1=@n48.Array Index) -> (Output:"CDH Instance Indices Struct Structure (by ref)")) -> (indices_3_#7:Set of Integers) {is=title K2Node_BreakStruct}
   ) -> (then, Result:Array of Integers)
   > RemoveInstances!(execute, self:Instanced Static Mesh Component Object Reference=@n48, InstancesToRemove:Array of Integers=@n49) -> (then, ReturnValue:bool)
   n50 = Break CDH Instance Info Struct(CDH_InstanceInfo_Struct=@n44) -> (radius_11_#4=~)
   n51 = Find(
    TargetMap:Map of Instanced Static Mesh Component Object References to CDH Instance Indices Struct Structures=Get l_finalInstancesToClearMap()
    Key:"Instanced Static Mesh Component Object Reference (by ref)"=@n50
   ) -> (Value:CDH Instance Indices Struct Structure)
   n52 = Branch(execute=@n44.LoopBody, Condition=@n51.ReturnValue)
   > n53 = Set l_deleteIndices(l_deleteIndices:Set of Integers=Break CDH Instance Indices Struct(CDH_InstanceIndices_Struct=@n51)) -> (Output_Get:Set of Integers)
   > Add!(execute, TargetSet:Set of Integers=@n53, NewItem:"Integer (by ref)"=@n50.instanceindex_5_#2) -> (then) {is=title K2Node_CallFunction}
   > Add!(
    execute
    TargetMap:Map of Instanced Static Mesh Component Object References to CDH Instance Indices Struct Structures=Get l_finalInstancesToClearMap()
    Key:"Instanced Static Mesh Component Object Reference (by ref)"=@n50
    Value:"CDH Instance Indices Struct Structure (by ref)"=Make CDH Instance Indices Struct!(indices_3_#7:Set of Integers=Get l_deleteIndices() -> (l_deleteIndices:Set of Integers)) -> (CDH_InstanceIndices_Struct:CDH Instance Indices Struct Structure) {is=title K2Node_MakeStruct}
   ) -> (then)
   Add(
    execute=@n52.else
    TargetMap=Get l_finalInstancesToClearMap()
    Key=@n50
    Value=Make CDH Instance Indices Struct(indices_3_#7=Make Set!("[0]":int=@n50.instanceindex_5_#2) -> (Set:Set of Integers) {is=title K2Node_MakeSet})
   )
   Return Node!(execute=@n2.then_4, totalHidden:int=Get l_total() -> (l_total=0))
   Vector Length(A=@n6.BoxExtent)
   "Max (Float)"!(A="Absolute (Float)"(A=@n15), B="Absolute (Float)"(A=@n15.Y), C:double=@n16) -> (ReturnValue)
   vector * vector(A:double=Get l_meshRadius() -> (l_meshRadius:double), B=@n14.l_curTransform_Scale) -> (ReturnValue=~)
   float * float(A=@n16)
  ]}
 IfDebugPreUpdateActors {nodes [
   n1 = IfDebugPreUpdateActors!() -> (then, Array:Array of BP Instanced Cluster Object References) {is=title K2Node_FunctionEntry}
   > n2 = Sequence()
   > Branch(Condition=Get IsTesting())
   > n3 = For Each Loop(Array:Array of BP Instanced Cluster Object References=@n1) -> (Array Element:BP Instanced Cluster Object Reference)
   > n4 = For Each Loop(Array:Array of Instanced Static Mesh Component Object References=Get Components By Class(self=@n3, ComponentClass=&InstancedStaticMeshComponent)) -> (Array Element:Instanced Static Mesh Component Object Reference)
   > SetEditorProperty(Object=@n4, PropertyName=NumCustomDataFloats, PropertyValue:int=MakeLiteralInt!(Value:int=1) -> (ReturnValue:int=0))
   Return Node!(execute=@n2.then_1, NeedDelay:bool=Get IsTesting())
  ]}
 SelectNearby {nodes [
   SelectNearby() {is=title K2Node_FunctionEntry}
   > n1 = GetSelectedLevelActors(self=Get EditorActorSubsystem())
   > n2 = Branch(Condition=Is Valid Index!(TargetArray:Array of Actor Object References=@n1, IndexToTest:int) -> (ReturnValue:bool))
   > n3 = Set l_loc(l_loc:Vector=Get Actor Location(self="Get (a copy)"(Array:Array of Actor Object References=@n1) -> (Output:Actor Object Reference))) -> (Output_Get:Vector)
   n4 = GetAllActorsOfClass(
    execute=[@n3.then, @n10.then]
    ActorClass=/Game/Environments/Asteria/Foliage/JunglePlants/FoliageCluster/BP_FoliageCluster.BP_FoliageCluster_C
   ) -> (OutActors:Array of BP Foliage Cluster Object References)
   > n5 = For Each Loop(Array:Array of BP Foliage Cluster Object References=@n4) -> (Array Element:BP Foliage Cluster Object Reference)
   n6 = Get l_toSelect() -> (l_toSelect:Array of BP Foliage Cluster Object References)
   Branch(execute=@n5.Completed, Condition="integer > integer"(A=Length(TargetArray:Array of BP Foliage Cluster Object References=@n6)))
   > SetSelectedLevelActors(self=Get EditorActorSubsystem(), ActorsToSelect=Get l_toSelect())
   n7 = vector * vector(A:Vector=Get l_loc() -> (l_loc:Vector), B="1,1,0") -> (ReturnValue=~)
   n8 = vector * vector(A=Get Actor Location(self=@n5), B="1,1,0") -> (ReturnValue=~)
   Branch(execute=@n5.LoopBody, Condition="float < float"(A="Distance (Vector)"(V1=@n8, V2=@n7), B=9999))
   > Add(TargetArray:Array of BP Foliage Cluster Object References=Get l_toSelect(), NewItem:"BP Foliage Cluster Object Reference (by ref)"=@n5) {is=title K2Node_CallArrayFunction}
   n9 = Get Editor Camera Info!(execute=@n2.else) -> (then, CameraLocation:Vector, ForwardDirection:Vector, FOV:float, ReturnValue:bool) {is=title K2Node_CallFunction}
   > n10 = Set l_loc(l_loc=@n9)
  ]}
 SetupTestData {nodes [
   n1 = SetupTestData!() -> (then, Object:Object Reference) {is=title K2Node_FunctionEntry}
   > Branch(Condition=Get IsTesting())
   > SetEditorProperty(Object=@n1, PropertyName=NumCustomDataFloats, PropertyValue=MakeLiteralInt(Value=1) -> (ReturnValue=0))
  ]}
 UpdateCustomData {nodes [
   n1 = UpdateCustomData!() -> (
    then
    comp:Instanced Static Mesh Component Object Reference
    InstanceIndex:int
    CustomDataValue:float=~
    bMarkRenderStateDirty:bool=~
    CustomDataIndex:int
   ) {is=title K2Node_FunctionEntry}
   > Branch(Condition=Get IsTesting())
   > SetCustomDataValue!(
    execute
    self:Instanced Static Mesh Component Object Reference=@n1
    InstanceIndex:int=@n1.InstanceIndex
    CustomDataIndex:int=@n1.CustomDataIndex
    CustomDataValue:float=@n1.CustomDataValue
    bMarkRenderStateDirty:bool=@n1.bMarkRenderStateDirty
   ) -> (then, ReturnValue:bool)
  ]}
 ValidateGoButtonColor {nodes [
   ValidateGoButtonColor() {is=title K2Node_FunctionEntry}
   > Set buttonColor_Default!(
    execute
    buttonColor_Default:Linear Color Structure=Select(
     Option 0:Linear Color Structure=(R=1,G=0.0098,B=0,A=1)
     Option 1:Linear Color Structure=(R=0.1383,G=0.651,B=0.0642,A=1)
     Index=Get IsTesting()
    ) -> (ReturnValue:Linear Color Structure)
    self:TA Base Editor Button Widget Object Reference=Get TA_Generic01_Button() -> (TA_Generic01_Button:TA Generic 01 Button Object Reference)
   ) -> (then, Output_Get:Linear Color Structure)
   > Set labelColor_Default!(
    execute
    labelColor_Default:Linear Color Structure=Select(Option 0=(R=1,G=0.9588,B=0.2402,A=1), Option 1=(R=0.9671,G=0.9899,B=1,A=1), Index=Get IsTesting())
    self:TA Base Editor Button Widget Object Reference=Get TA_Generic01_Button()
   ) -> (then, Output_Get:Linear Color Structure)
  ]}
 ValidateGoButtonText {nodes [
   ValidateGoButtonText() {is=title K2Node_FunctionEntry}
   > n1 = Sequence!(execute) -> (then_0, then_1, then_2:Exec, then_3:Exec, then_4:Exec)
   > Set l_string(l_string:String="Find Overlapping Instances on ") -> (Output_Get:String)
   n2 = Get RunOnlyOnSelectedActors()
   Set l_string(
    execute=@n1.then_1
    l_string=Append!(
     A=Get l_string() -> (l_string:String)
     B=Select(Option 0:String=All Cluster Actors, Option 1:String=Selected Cluster Actors, Index=@n2) -> (ReturnValue:String)
    ) -> (ReturnValue)
   )
   n3 = Get IsTesting()
   Set l_string(
    execute=@n1.then_2
    l_string=Append!(A=Get l_string(), B="\r\n(", C:String=Select(Option 0=And Delete Them Permanently, Option 1=And Debug Them, Index=@n3), D:String=")") -> (ReturnValue)
   )
   ValidateGoButtonColor(execute=@n1.then_3) {is=title K2Node_CallFunction}
   setLabel!(execute=@n1.then_4, self:TA Base Editor Button Widget Object Reference=Get TA_Generic01_Button(), label:String=Get l_string()) -> (then)
  ]}
 ValidateSelectText {nodes [
   ValidateSelectText() {is=title K2Node_FunctionEntry}
   > n1 = Sequence()
   > Set l_string(l_string="Select All Clusters Near ")
   n2 = GetSelectedActors(execute=@n1.then_1)
   > n3 = Branch(Condition=Is Valid Index(TargetArray=@n2))
   n4 = "Get (a copy)"(Array=@n2)
   > Set l_string(l_string=Append!(A=Get l_string(), B=GetActorLabel(self=@n4, bCreateIfNone=true)) -> (ReturnValue))
   Set l_string(execute=@n3.else, l_string=Append(A=Get l_string(), B=Camera))
   setLabel(execute=@n1.then_4, self=Get btnSelect() -> (btnSelect:TA Generic 01 Button Object Reference), label=Get l_string())
  ]}
 ValidateVisuals {nodes [
   ValidateVisuals() {is=title K2Node_FunctionEntry}
   > ValidateGoButtonText() {is=title K2Node_CallFunction}
   > ValidateSelectText() {is=title K2Node_CallFunction}
  ]}
]
variables [
 IsTesting {type=boolean, default=true}
 OverlapPercentThreshold {type=number, default=0.45}
 RunOnlyOnSelectedActors {type=boolean, default=true}
]
widgets {
 info {namedSlotCount=2, rootWidgetClass=&VerticalBox, widgetCount=13}
 widgets [
  VerticalBox_0 {widgetClassPath=&VerticalBox, widget=VerticalBox_0}
  HorizontalBox_0 {
   widgetClassPath=&HorizontalBox
   parent=VerticalBox_0
   widget=HorizontalBox_0
   slot {class=VerticalBoxSlot, properties {padding=<ltrb 0 0 0 10>}}
  }
  btnSelect {
   widgetClassPath=&TA_Generic01_Button
   parent=HorizontalBox_0
   widget=btnSelect
   properties {
    label=Select Nearby
    toolTipText=Select all foliage cluster actors near the selected Actor. If no actor is selected then it selects nearest the camera in X/Y
   }
   slot {class=HorizontalBoxSlot, properties {size {sizeRule=Fill, value=1}}}
  }
  CommonTextBlock {
   widgetClassPath=&CommonTextBlock
   bIsVariable=true
   parent=HorizontalBox_0
   widget=CommonTextBlock
   properties {
    text=Radius:
    colorAndOpacity {colorUseRule=UseColor_Specified, specifiedColor=<rgba 0.6979 0.6979 0.6979 1>}
    font {
     bForceMonospaced=false
     bMaterialIsStencil=false
     fontMaterial=null
     fontObject=&NotoSans
     letterSpacing=0
     monospacedWidth=1
     size=14
     skewAmount=0
     typefaceFontName=NotoSans-Regular
     outlineSettings {
      bApplyOutlineToDropShadows=false
      bMiteredCorners=false
      bSeparateFillAlpha=false
      outlineMaterial=null
      outlineSize=0
      outlineColor=<rgba 0 0 0 1>
     }
    }
    shadowOffset=<0 0>
   }
   slot {class=HorizontalBoxSlot, properties {padding=<ltrb 5 0 0 0>}}
  }
  SpinBox_87 {
   widgetClassPath=/Script/UMG.SpinBox
   bIsVariable=true
   parent=HorizontalBox_0
   widget=SpinBox_87
   properties {
    bOverride_MaxSliderValue=true
    bOverride_MinSliderValue=true
    bOverride_MinValue=true
    delta=10
    maxFractionalDigits=0
    maxSliderValue=20000
    minFractionalDigits=0
    minSliderValue=1500
    value=10000
    foregroundColor {colorUseRule=UseColor_Specified, specifiedColor=<rgba 1 1 1 1>}
    widgetStyle {
     activeBackgroundBrush {
      bIsDynamicallyLoaded=false
      drawAs=RoundedBox
      imageType=NoImage
      mirroring=NoMirror
      resourceName=None
      resourceObject=null
      tiling=NoTile
      imageSize=<0 0>
      margin=<ltrb 0 0 0 0>
      outlineSettings {
       bUseBrushTransparency=false
       roundingType=FixedRadius
       width=1
       color {colorUseRule=UseColor_ColorTable, specifiedColor=<rgba 0 0 0 0>}
       cornerRadii=<4 4 4 4>
      }
      tintColor {colorUseRule=UseColor_Specified, specifiedColor=<rgba 0.0048 0.0048 0.0048 1>}
      uVRegion {bIsValid=false, max=<0 0>, min=<0 0>}
     }
     activeFillBrush {
      bIsDynamicallyLoaded=false
      drawAs=RoundedBox
      imageType=NoImage
      mirroring=NoMirror
      resourceName=None
      resourceObject=null
      tiling=NoTile
      imageSize=<0 0>
      margin=<ltrb 0 0 0 0>
      outlineSettings {
       bUseBrushTransparency=false
       roundingType=FixedRadius
       width=1
       color {colorUseRule=UseColor_Specified, specifiedColor=<rgba 0 0 0 0>}
       cornerRadii=<4 4 4 4>
      }
      tintColor {colorUseRule=UseColor_Specified, specifiedColor=<rgba 0 0.162 0.7454 1>}
      uVRegion {bIsValid=false, max=<0 0>, min=<0 0>}
     }
     arrowsImage {
      bIsDynamicallyLoaded=false
      drawAs=NoDrawType
      imageType=NoImage
      mirroring=NoMirror
      resourceName=None
      resourceObject=null
      tiling=NoTile
      imageSize=<0 0>
      margin=<ltrb 0 0 0 0>
      outlineSettings {
       bUseBrushTransparency=false
       roundingType=HalfHeightRadius
       width=0
       color {colorUseRule=UseColor_Specified, specifiedColor=<rgba 0 0 0 0>}
       cornerRadii=<0 0 0 0>
      }
      tintColor {colorUseRule=UseColor_Specified, specifiedColor=<rgba 1 1 1 1>}
      uVRegion {bIsValid=false, max=<0 0>, min=<0 0>}
     }
     backgroundBrush {
      bIsDynamicallyLoaded=false
      drawAs=RoundedBox
      imageType=NoImage
      mirroring=NoMirror
      resourceName=None
      resourceObject=null
      tiling=NoTile
      imageSize=<0 0>
      margin=<ltrb 0 0 0 0>
      outlineSettings {
       bUseBrushTransparency=false
       roundingType=FixedRadius
       width=1
       color {colorUseRule=UseColor_ColorTable, specifiedColor=<rgba 0 0 0 0>}
       cornerRadii=<4 4 4 4>
      }
      tintColor {colorUseRule=UseColor_Specified, specifiedColor=<rgba 0.0048 0.0048 0.0048 1>}
      uVRegion {bIsValid=false, max=<0 0>, min=<0 0>}
     }
     foregroundColor {colorUseRule=UseColor_Specified, specifiedColor=<rgba 1 1 1 1>}
     hoveredBackgroundBrush {
      bIsDynamicallyLoaded=false
      drawAs=RoundedBox
      imageType=NoImage
      mirroring=NoMirror
      resourceName=None
      resourceObject=null
      tiling=NoTile
      imageSize=<0 0>
      margin=<ltrb 0 0 0 0>
      outlineSettings {
       bUseBrushTransparency=false
       roundingType=FixedRadius
       width=1
       color {colorUseRule=UseColor_ColorTable, specifiedColor=<rgba 0 0 0 0>}
       cornerRadii=<4 4 4 4>
      }
      tintColor {colorUseRule=UseColor_Specified, specifiedColor=<rgba 0.0048 0.0048 0.0048 1>}
      uVRegion {bIsValid=false, max=<0 0>, min=<0 0>}
     }
     hoveredFillBrush {
      bIsDynamicallyLoaded=false
      drawAs=RoundedBox
      imageType=NoImage
      mirroring=NoMirror
      resourceName=None
      resourceObject=null
      tiling=NoTile
      imageSize=<0 0>
      margin=<ltrb 0 0 0 0>
      outlineSettings {
       bUseBrushTransparency=false
       roundingType=FixedRadius
       width=1
       color {colorUseRule=UseColor_Specified, specifiedColor=<rgba 0 0 0 0>}
       cornerRadii=<4 4 4 4>
      }
      tintColor {colorUseRule=UseColor_Specified, specifiedColor=<rgba 0.0953 0.0953 0.0953 1>}
      uVRegion {bIsValid=false, max=<0 0>, min=<0 0>}
     }
     inactiveFillBrush {
      bIsDynamicallyLoaded=false
      drawAs=RoundedBox
      imageType=NoImage
      mirroring=NoMirror
      resourceName=None
      resourceObject=null
      tiling=NoTile
      imageSize=<0 0>
      margin=<ltrb 0 0 0 0>
      outlineSettings {
       bUseBrushTransparency=false
       roundingType=FixedRadius
       width=1
       color {colorUseRule=UseColor_Specified, specifiedColor=<rgba 0 0 0 0>}
       cornerRadii=<4 4 4 4>
      }
      tintColor {colorUseRule=UseColor_Specified, specifiedColor=<rgba 0.0395 0.0395 0.0395 1>}
      uVRegion {bIsValid=false, max=<0 0>, min=<0 0>}
     }
     insetPadding=<ltrb 1 1 1 1>
     textPadding=<ltrb 7 2.5 7 2.5>
    }
   }
   slot {class=HorizontalBoxSlot, properties {padding=<ltrb 0 0 10 0>}}
  }
  Image_73 {
   widgetClassPath=&Image
   bIsVariable=true
   parent=VerticalBox_0
   widget=Image_73
   properties {brush {
     bIsDynamicallyLoaded=false
     drawAs=Image
     imageType=NoImage
     mirroring=NoMirror
     resourceName=None
     resourceObject=null
     tiling=NoTile
     imageSize=<32 2>
     margin=<ltrb 0 0 0 0>
     outlineSettings {
      bUseBrushTransparency=false
      roundingType=HalfHeightRadius
      width=0
      color {colorUseRule=UseColor_Specified, specifiedColor=<rgba 0 0 0 0>}
      cornerRadii=<0 0 0 0>
     }
     tintColor {colorUseRule=UseColor_Specified, specifiedColor=<rgba 1 1 1 1>}
     uVRegion {bIsValid=false, max=<0 0>, min=<0 0>}
    }}
   slot {class=VerticalBoxSlot, properties {padding=<ltrb 0 0 0 10>}}
  }
  DetailsView_0 {
   widgetClassPath=&DetailsView
   parent=VerticalBox_0
   widget=DetailsView_0
   properties {categoriesToShow [nocategory]}
   slot {class=VerticalBoxSlot}
  }
  TA_Generic01_Button {
   widgetClassPath=&TA_Generic01_Button
   parent=VerticalBox_0
   widget=TA_Generic01_Button
   properties {
    label="Find Overlapping Instances on Selected Clusters\r\n(And delete them)"
    buttonColor_Default=<rgba 0.8333 0.0394 0 1>
    labelColor_Default=<rgba 1 0.9034 0 1>
   }
   slot {class=VerticalBoxSlot}
  }
  CommonTextBlock_0 {
   widgetClassPath=&CommonTextBlock
   bIsVariable=true
   parent=VerticalBox_0
   widget=CommonTextBlock_0
   properties {
    colorAndOpacity {colorUseRule=UseColor_Specified, specifiedColor=<rgba 0.6979 0.6979 0.6979 1>}
    font {
     bForceMonospaced=false
     bMaterialIsStencil=false
     fontMaterial=null
     fontObject=&NotoSans
     letterSpacing=0
     monospacedWidth=1
     size=14
     skewAmount=0
     typefaceFontName=NotoSans-Regular
     outlineSettings {
      bApplyOutlineToDropShadows=false
      bMiteredCorners=false
      bSeparateFillAlpha=false
      outlineMaterial=null
      outlineSize=0
      outlineColor=<rgba 0 0 0 1>
     }
    }
    shadowOffset=<0 0>
   }
   slot {class=VerticalBoxSlot}
  }
  Border_0 {
   widgetClassPath=&Border
   parent=VerticalBox_0
   widget=Border_0
   properties {brushColor=<rgba 0 0 0 0.5>}
   slot {class=VerticalBoxSlot}
  }
  ExpandableArea_103 {
   widgetClassPath=/Script/UMG.ExpandableArea
   bIsVariable=true
   parent=Border_0
   widget=ExpandableArea_103
   properties {
    borderBrush {
     bIsDynamicallyLoaded=false
     drawAs=RoundedBox
     imageType=NoImage
     mirroring=NoMirror
     resourceName=None
     resourceObject=null
     tiling=NoTile
     imageSize=<0 0>
     margin=<ltrb 0 0 0 0>
     outlineSettings {
      bUseBrushTransparency=false
      roundingType=FixedRadius
      width=0
      color {colorUseRule=UseColor_Specified, specifiedColor=<rgba 0 0 0 0>}
      cornerRadii=<4 4 4 4>
     }
     tintColor {colorUseRule=UseColor_Specified, specifiedColor=<rgba 0.0176 0.0176 0.0176 1>}
     uVRegion {bIsValid=false, max=<0 0>, min=<0 0>}
    }
    style {
     rolloutAnimationSeconds=0.1
     collapsedImage {
      bIsDynamicallyLoaded=false
      drawAs=Image
      imageType=Vector
      mirroring=NoMirror
      resourceName=../../../Engine/Content/Slate/Starship/Common/caret-right.svg
      resourceObject=null
      tiling=NoTile
      imageSize=<16 16>
      margin=<ltrb 0 0 0 0>
      outlineSettings {
       bUseBrushTransparency=false
       roundingType=HalfHeightRadius
       width=0
       color {colorUseRule=UseColor_Specified, specifiedColor=<rgba 0 0 0 0>}
       cornerRadii=<0 0 0 0>
      }
      tintColor {colorUseRule=UseColor_Specified, specifiedColor=<rgba 0.5271 0.5271 0.5271 1>}
      uVRegion {bIsValid=false, max=<0 0>, min=<0 0>}
     }
     expandedImage {
      bIsDynamicallyLoaded=false
      drawAs=Image
      imageType=Vector
      mirroring=NoMirror
      resourceName=../../../Engine/Content/Slate/Starship/Common/caret-down.svg
      resourceObject=null
      tiling=NoTile
      imageSize=<16 16>
      margin=<ltrb 0 0 0 0>
      outlineSettings {
       bUseBrushTransparency=false
       roundingType=HalfHeightRadius
       width=0
       color {colorUseRule=UseColor_Specified, specifiedColor=<rgba 0 0 0 0>}
       cornerRadii=<0 0 0 0>
      }
      tintColor {colorUseRule=UseColor_Specified, specifiedColor=<rgba 0.5271 0.5271 0.5271 1>}
      uVRegion {bIsValid=false, max=<0 0>, min=<0 0>}
     }
    }
   }
   slot {class=BorderSlot}
  }
  CommonTextBlock_2 {
   widgetClassPath=&CommonTextBlock
   namedSlotHost=ExpandableArea_103
   widget=CommonTextBlock_2
   properties {
    text=HELP
    colorAndOpacity {colorUseRule=UseColor_Specified, specifiedColor=<rgba 0.6979 0.6979 0.6979 1>}
    font {
     bForceMonospaced=false
     bMaterialIsStencil=false
     fontMaterial=null
     fontObject=&NotoSans
     letterSpacing=0
     monospacedWidth=1
     size=14
     skewAmount=0
     typefaceFontName=NotoSans-Bold
     outlineSettings {
      bApplyOutlineToDropShadows=false
      bMiteredCorners=false
      bSeparateFillAlpha=false
      outlineMaterial=null
      outlineSize=0
      outlineColor=<rgba 0 0 0 1>
     }
    }
    shadowOffset=<0 0>
   }
  }
  MultiLineEditableTextBox_76 {
   widgetClassPath=/Script/UMG.MultiLineEditableTextBox
   bIsVariable=true
   namedSlotHost=ExpandableArea_103
   widget=MultiLineEditableTextBox_76
   properties {
    bIsReadOnly=true
    text="Run with IsTesting=True to be in Debug Mode.  This is only effective if you have connected the optional debug nodes in the material M_FoliageMaster_2023.\r\n\r\nIf not testing, this tool will actually delete instances on the affected actors so run only if you are sure.\r\n\r\nWe need to regenerate the final prop groups if we alter clusters."
    widgetStyle {
     backgroundColor {colorUseRule=UseColor_Specified, specifiedColor=<rgba 1 1 1 1>}
     backgroundImageFocused {
      bIsDynamicallyLoaded=false
      drawAs=RoundedBox
      imageType=NoImage
      mirroring=NoMirror
      resourceName=None
      resourceObject=null
      tiling=NoTile
      imageSize=<0 0>
      margin=<ltrb 0 0 0 0>
      outlineSettings {
       bUseBrushTransparency=false
       roundingType=FixedRadius
       width=1
       color {colorUseRule=UseColor_ColorTable, specifiedColor=<rgba 0 0 0 0>}
       cornerRadii=<4 4 4 4>
      }
      tintColor {colorUseRule=UseColor_Specified, specifiedColor=<rgba 0.0048 0.0048 0.0048 1>}
      uVRegion {bIsValid=false, max=<0 0>, min=<0 0>}
     }
     backgroundImageHovered {
      bIsDynamicallyLoaded=false
      drawAs=RoundedBox
      imageType=NoImage
      mirroring=NoMirror
      resourceName=None
      resourceObject=null
      tiling=NoTile
      imageSize=<0 0>
      margin=<ltrb 0 0 0 0>
      outlineSettings {
       bUseBrushTransparency=false
       roundingType=FixedRadius
       width=1
       color {colorUseRule=UseColor_ColorTable, specifiedColor=<rgba 0 0 0 0>}
       cornerRadii=<4 4 4 4>
      }
      tintColor {colorUseRule=UseColor_Specified, specifiedColor=<rgba 0.0048 0.0048 0.0048 1>}
      uVRegion {bIsValid=false, max=<0 0>, min=<0 0>}
     }
     backgroundImageNormal {
      bIsDynamicallyLoaded=false
      drawAs=RoundedBox
      imageType=NoImage
      mirroring=NoMirror
      resourceName=None
      resourceObject=null
      tiling=NoTile
      imageSize=<0 0>
      margin=<ltrb 0 0 0 0>
      outlineSettings {
       bUseBrushTransparency=false
       roundingType=FixedRadius
       width=1
       color {colorUseRule=UseColor_ColorTable, specifiedColor=<rgba 0 0 0 0>}
       cornerRadii=<4 4 4 4>
      }
      tintColor {colorUseRule=UseColor_Specified, specifiedColor=<rgba 0.0048 0.0048 0.0048 1>}
      uVRegion {bIsValid=false, max=<0 0>, min=<0 0>}
     }
     backgroundImageReadOnly {
      bIsDynamicallyLoaded=false
      drawAs=RoundedBox
      imageType=NoImage
      mirroring=NoMirror
      resourceName=None
      resourceObject=null
      tiling=NoTile
      imageSize=<0 0>
      margin=<ltrb 0 0 0 0>
      outlineSettings {
       bUseBrushTransparency=false
       roundingType=FixedRadius
       width=1
       color {colorUseRule=UseColor_ColorTable, specifiedColor=<rgba 0 0 0 0>}
       cornerRadii=<4 4 4 4>
      }
      tintColor {colorUseRule=UseColor_Specified, specifiedColor=<rgba 0.0284 0.0284 0.0284 1>}
      uVRegion {bIsValid=false, max=<0 0>, min=<0 0>}
     }
     focusedForegroundColor {colorUseRule=UseColor_Specified, specifiedColor=<rgba 1 1 1 1>}
     foregroundColor {colorUseRule=UseColor_Specified, specifiedColor=<rgba 0.5271 0.5271 0.5271 1>}
     hScrollBarPadding=<ltrb 0 0 0 0>
     padding=<ltrb 8 3 8 4>
     readOnlyForegroundColor {colorUseRule=UseColor_Foreground, specifiedColor=<rgba 1 0 1 1>}
     scrollBarStyle {
      thickness=8
      draggedThumbImage {
       bIsDynamicallyLoaded=false
       drawAs=RoundedBox
       imageType=NoImage
       mirroring=NoMirror
       resourceName=None
       resourceObject=null
       tiling=NoTile
       imageSize=<0 0>
       margin=<ltrb 0 0 0 0>
       outlineSettings {
        bUseBrushTransparency=false
        roundingType=FixedRadius
        width=0
        color {colorUseRule=UseColor_Specified, specifiedColor=<rgba 0 0 0 0>}
        cornerRadii=<4 4 4 4>
       }
       tintColor {colorUseRule=UseColor_Specified, specifiedColor=<rgba 0.2159 0.2159 0.2159 1>}
       uVRegion {bIsValid=false, max=<0 0>, min=<0 0>}
      }
      horizontalBackgroundImage {
       bIsDynamicallyLoaded=false
       drawAs=NoDrawType
       imageType=NoImage
       mirroring=NoMirror
       resourceName=None
       resourceObject=null
       tiling=NoTile
       imageSize=<0 0>
       margin=<ltrb 0 0 0 0>
       outlineSettings {
        bUseBrushTransparency=false
        roundingType=HalfHeightRadius
        width=0
        color {colorUseRule=UseColor_Specified, specifiedColor=<rgba 0 0 0 0>}
        cornerRadii=<0 0 0 0>
       }
       tintColor {colorUseRule=UseColor_Specified, specifiedColor=<rgba 1 1 1 1>}
       uVRegion {bIsValid=false, max=<0 0>, min=<0 0>}
      }
      horizontalBottomSlotImage {
       bIsDynamicallyLoaded=false
       drawAs=NoDrawType
       imageType=NoImage
       mirroring=NoMirror
       resourceName=None
       resourceObject=null
       tiling=NoTile
       imageSize=<0 0>
       margin=<ltrb 0 0 0 0>
       outlineSettings {
        bUseBrushTransparency=false
        roundingType=HalfHeightRadius
        width=0
        color {colorUseRule=UseColor_Specified, specifiedColor=<rgba 0 0 0 0>}
        cornerRadii=<0 0 0 0>
       }
       tintColor {colorUseRule=UseColor_Specified, specifiedColor=<rgba 1 1 1 1>}
       uVRegion {bIsValid=false, max=<0 0>, min=<0 0>}
      }
      horizontalTopSlotImage {
       bIsDynamicallyLoaded=false
       drawAs=NoDrawType
       imageType=NoImage
       mirroring=NoMirror
       resourceName=None
       resourceObject=null
       tiling=NoTile
       imageSize=<0 0>
       margin=<ltrb 0 0 0 0>
       outlineSettings {
        bUseBrushTransparency=false
        roundingType=HalfHeightRadius
        width=0
        color {colorUseRule=UseColor_Specified, specifiedColor=<rgba 0 0 0 0>}
        cornerRadii=<0 0 0 0>
       }
       tintColor {colorUseRule=UseColor_Specified, specifiedColor=<rgba 1 1 1 1>}
       uVRegion {bIsValid=false, max=<0 0>, min=<0 0>}
      }
      hoveredThumbImage {
       bIsDynamicallyLoaded=false
       drawAs=RoundedBox
       imageType=NoImage
       mirroring=NoMirror
       resourceName=None
       resourceObject=null
       tiling=NoTile
       imageSize=<0 0>
       margin=<ltrb 0 0 0 0>
       outlineSettings {
        bUseBrushTransparency=false
        roundingType=FixedRadius
        width=0
        color {colorUseRule=UseColor_Specified, specifiedColor=<rgba 0 0 0 0>}
        cornerRadii=<4 4 4 4>
       }
       tintColor {colorUseRule=UseColor_Specified, specifiedColor=<rgba 0.2159 0.2159 0.2159 1>}
       uVRegion {bIsValid=false, max=<0 0>, min=<0 0>}
      }
      normalThumbImage {
       bIsDynamicallyLoaded=false
       drawAs=RoundedBox
       imageType=NoImage
       mirroring=NoMirror
       resourceName=None
       resourceObject=null
       tiling=NoTile
       imageSize=<0 0>
       margin=<ltrb 0 0 0 0>
       outlineSettings {
        bUseBrushTransparency=false
        roundingType=FixedRadius
        width=0
        color {colorUseRule=UseColor_Specified, specifiedColor=<rgba 0 0 0 0>}
        cornerRadii=<4 4 4 4>
       }
       tintColor {colorUseRule=UseColor_Specified, specifiedColor=<rgba 0.0953 0.0953 0.0953 1>}
       uVRegion {bIsValid=false, max=<0 0>, min=<0 0>}
      }
      verticalBackgroundImage {
       bIsDynamicallyLoaded=false
       drawAs=NoDrawType
       imageType=NoImage
       mirroring=NoMirror
       resourceName=None
       resourceObject=null
       tiling=NoTile
       imageSize=<0 0>
       margin=<ltrb 0 0 0 0>
       outlineSettings {
        bUseBrushTransparency=false
        roundingType=HalfHeightRadius
        width=0
        color {colorUseRule=UseColor_Specified, specifiedColor=<rgba 0 0 0 0>}
        cornerRadii=<0 0 0 0>
       }
       tintColor {colorUseRule=UseColor_Specified, specifiedColor=<rgba 1 1 1 1>}
       uVRegion {bIsValid=false, max=<0 0>, min=<0 0>}
      }
      verticalBottomSlotImage {
       bIsDynamicallyLoaded=false
       drawAs=NoDrawType
       imageType=NoImage
       mirroring=NoMirror
       resourceName=None
       resourceObject=null
       tiling=NoTile
       imageSize=<0 0>
       margin=<ltrb 0 0 0 0>
       outlineSettings {
        bUseBrushTransparency=false
        roundingType=HalfHeightRadius
        width=0
        color {colorUseRule=UseColor_Specified, specifiedColor=<rgba 0 0 0 0>}
        cornerRadii=<0 0 0 0>
       }
       tintColor {colorUseRule=UseColor_Specified, specifiedColor=<rgba 1 1 1 1>}
       uVRegion {bIsValid=false, max=<0 0>, min=<0 0>}
      }
      verticalTopSlotImage {
       bIsDynamicallyLoaded=false
       drawAs=NoDrawType
       imageType=NoImage
       mirroring=NoMirror
       resourceName=None
       resourceObject=null
       tiling=NoTile
       imageSize=<0 0>
       margin=<ltrb 0 0 0 0>
       outlineSettings {
        bUseBrushTransparency=false
        roundingType=HalfHeightRadius
        width=0
        color {colorUseRule=UseColor_Specified, specifiedColor=<rgba 0 0 0 0>}
        cornerRadii=<0 0 0 0>
       }
       tintColor {colorUseRule=UseColor_Specified, specifiedColor=<rgba 1 1 1 1>}
       uVRegion {bIsValid=false, max=<0 0>, min=<0 0>}
      }
     }
     textStyle {
      overflowPolicy=Clip
      transformPolicy=None
      colorAndOpacity {colorUseRule=UseColor_Specified, specifiedColor=<rgba 1 1 1 1>}
      font {
       bForceMonospaced=false
       bMaterialIsStencil=false
       fontMaterial=null
       fontObject=null
       letterSpacing=0
       monospacedWidth=1
       size=14
       skewAmount=0
       typefaceFontName=Regular
       outlineSettings {
        bApplyOutlineToDropShadows=false
        bMiteredCorners=false
        bSeparateFillAlpha=false
        outlineMaterial=null
        outlineSize=0
        outlineColor=<rgba 0 0 0 1>
       }
      }
      highlightColor {colorUseRule=UseColor_ColorTable, specifiedColor=<rgba 0 0 0 0>}
      highlightShape {
       bIsDynamicallyLoaded=false
       drawAs=Image
       imageType=NoImage
       mirroring=NoMirror
       resourceName=None
       resourceObject=null
       tiling=NoTile
       imageSize=<0 0>
       margin=<ltrb 0 0 0 0>
       outlineSettings {
        bUseBrushTransparency=false
        roundingType=HalfHeightRadius
        width=0
        color {colorUseRule=UseColor_Specified, specifiedColor=<rgba 0 0 0 0>}
        cornerRadii=<0 0 0 0>
       }
       tintColor {colorUseRule=UseColor_ColorTable, specifiedColor=<rgba 0 0 0 0>}
       uVRegion {bIsValid=false, max=<0 0>, min=<0 0>}
      }
      selectedBackgroundColor {colorUseRule=UseColor_ColorTable, specifiedColor=<rgba 0 0 0 0>}
      shadowColorAndOpacity=<rgba 0 0 0 1>
      shadowOffset=<0 0>
      strikeBrush {
       bIsDynamicallyLoaded=false
       drawAs=Image
       imageType=NoImage
       mirroring=NoMirror
       resourceName=None
       resourceObject=null
       tiling=NoTile
       imageSize=<32 32>
       margin=<ltrb 0 0 0 0>
       outlineSettings {
        bUseBrushTransparency=false
        roundingType=HalfHeightRadius
        width=0
        color {colorUseRule=UseColor_Specified, specifiedColor=<rgba 0 0 0 0>}
        cornerRadii=<0 0 0 0>
       }
       tintColor {colorUseRule=UseColor_Specified, specifiedColor=<rgba 1 1 1 1>}
       uVRegion {bIsValid=false, max=<0 0>, min=<0 0>}
      }
      underlineBrush {
       bIsDynamicallyLoaded=false
       drawAs=Image
       imageType=NoImage
       mirroring=NoMirror
       resourceName=None
       resourceObject=null
       tiling=NoTile
       imageSize=<32 32>
       margin=<ltrb 0 0 0 0>
       outlineSettings {
        bUseBrushTransparency=false
        roundingType=HalfHeightRadius
        width=0
        color {colorUseRule=UseColor_Specified, specifiedColor=<rgba 0 0 0 0>}
        cornerRadii=<0 0 0 0>
       }
       tintColor {colorUseRule=UseColor_Specified, specifiedColor=<rgba 1 1 1 1>}
       uVRegion {bIsValid=false, max=<0 0>, min=<0 0>}
      }
     }
     vScrollBarPadding=<ltrb 0 0 0 0>
    }
   }
  }
 ]
}
